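; Landing knockback gameplay ability for BarnTowelArmory environmental items; this graph holds only its FindLauncher helper, which resolves the actor treated as the launcher.
; FindLauncher reads the effect context that granted the ability and returns its instigator actor when valid.
; Otherwise it returns the DefaultLauncher actor passed in by the caller.
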(fn FindLauncher
  (params (DefaultLauncher :type Actor%20Object%20Reference))
  (bind _returnvalue (Ability|GetGrantedbyEffectContext))
  (bind _returnvalue_2 (Ability|EffectContext|GetInstigatorActor _returnvalue))
  (Utilities|IsValid _returnvalue_2
    (:Is%20Valid
      (return _returnvalue_2))
    (:Is%20Not%20Valid
      (return DefaultLauncher))))
